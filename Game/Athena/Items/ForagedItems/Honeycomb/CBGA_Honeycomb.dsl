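; Honeycomb foraged item actor for Athena that restricts which pawns can interact with it.
; BlueprintCanInteract casts the interacting pawn to a player pawn, stores it in a local variable, and refuses if that cast fails.
; Interaction is allowed only when that pawn lacks the SpookyMist foraged-item gameplay tag, is not a remote-controlled pawn, and is not controlling one.
; The construction script only calls its parent.
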(fn BlueprintCanInteract
  (params (InteractingPawn :type Fort%20Pawn%20Object%20Reference) (InteractionBeingAttempted :type EInteractionBeingAttempted%20Enum :default "FirstInteraction") (InteractionType :type TInteractionType%20Enum :default "IT_NoInteraction"))
  (bind (_asfort_remote_controlled_pawn_athena _bsuccess) (Utilities|Casting|CastToFortRemoteControlledPawnAthena (LocalVariables|GetPlayerPawn)))
  (bind _asfort_player_pawn (Utilities|Casting|CastToFortPlayerPawn InteractingPawn)
    (:then
      (LocalVariables|SetPlayerPawn _asfort_player_pawn)
      (return (Math|Boolean|ANDBoolean (not (GameplayTags|HasMatchingGameplayTag (LocalVariables|GetPlayerPawn) "(TagName=\"GameplayEffect.ForagedItem.SpookyMist\")")) (not _bsuccess) (not (RemoteControlPawn|HasControlledRCPawn (LocalVariables|GetPlayerPawn))))))
    (:CastFailed
      (return false))))

(fn ConstructionScript
  (|CallParentFunction))
