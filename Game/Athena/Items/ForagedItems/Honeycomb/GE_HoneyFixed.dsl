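; A foraged Honeycomb gameplay ability, despite its GE_ prefix, that applies a consumable effect to its owner.
; Activation prints a debug "Hello" and runs a montage play-and-wait-for-trigger task.
; A DoOnce gate then applies GE_Athena_PurpleStuffB, borrowed from the Coconut item folder, and ends the ability.
; The end-ability event has no logic.
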(event Ability|EventActivateAbility
  (Development|PrintString "Hello")
  ;[1 1 1] Play montage
  (Ability|Tasks|PlayMontageInfoAndWaitForTrigger "None" (Variables|Animation|GetMontageInfo) "SimulateOnServer"
    (:Triggered
      _)
    (:Cancelled
      _)
    (:Completed
      _))
  (Utilities|FlowControl|DoOnce false
    (:Completed
      (Ability|ApplyGameplayEffectToOwner "/Game/Athena/Items/ForagedItems/Coconut/GE_Athena_PurpleStuffB.GE_Athena_PurpleStuffB_C")
      (Ability|EndAbility))))

(event Ability|EventOnEndAbility
  (params (bWasCancelled :type Boolean)))
